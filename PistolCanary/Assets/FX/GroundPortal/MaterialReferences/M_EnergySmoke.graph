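# M_EnergySmoke is a translucent, unlit, two-sided energy smoke material for Niagara mesh particles in ground portal effects.
# Panning flame noise, warped by reflection-vector distortion, drives tinted color, while fresnel, view-angle and depth fades, plasma erosion and a V gradient shape opacity.
# A GameTime-stamped hit wave subtracts two expanding world-space sphere masks into a shockwave ring, zeroed on forward and mobile.
# Vertex offset combines hit-wave displacement with particle-scaled panning noise.
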Material /PistolCanary/Assets/FX/GroundPortal/MaterialReferences/M_EnergySmoke
graph {
 properties {
  bUsedWithNiagaraMeshParticles=true
  blendMode=BLEND_Translucent
  pixelDepthOffsetMode=PDOM_Legacy
  shadingModel=MSM_Unlit
  twoSided=true
  usedShadingModels=Unlit
 }
 nodes [
  Constant() {r=50}
  Divide()
  MaterialFunctionCall() {materialFunction=&DebugTimeSine}
  n1 = ParticleColor()
  n2 = VectorParameter() {parameterName=ColorTint, defaultValue=<rgba 0.5267 0.0246 0.6823 1>}
  n3 = Multiply(A=TextureCoordinate(), B=AppendVector(A=Constant() {r=2}, B=Constant() {r=0.6}))
  n4 = TextureCoordinate()
  n5 = AppendVector(A=Constant() {r=2}, B=Constant() {r=0.6})
  n6 = TextureSample(UVs=Panner(Coordinate=Multiply(A=@n4, B=@n5)) {speedX=-0.02, speedY=-0.05}) {samplerType=SAMPLERTYPE_LinearColor, texture=/Game/Effects/Fort_Effects/Effects/Environment/Desert/Dust/Organize/T_Turbulence_Seamless}
  n7 = Divide(A=ComponentMask(None=ReflectionVectorWS(CustomWorldNormal=@n6.RGB)) {g=true, r=true}) {constB=5}
  n8 = TextureSample(UVs=Add(A=Panner(Coordinate=@n3) {speedX=0.025, speedY=0.075}, B=@n7)) {samplerType=SAMPLERTYPE_LinearGrayscale, texture=/Game/Weapons/FORT_Melee/Pickaxe_Cube_Ninja_Male/FX/T_FluidFlameTile}
  n9 = VectorParameter() {group=HitWave, parameterName=HitWave-Tint, defaultValue=<rgba 0.2236 2 0.9691 1>}
  n10 = MaterialFunctionCall(Distance=Constant() {r=2000}) {materialFunction=&Near_Camera_Fade}
  n11 = OneMinus(None=NamedRerouteUsage())
  n12 = VectorParameter() {group=HitWave, parameterName=Hitwave-AddBrighness, defaultValue=<rgba 0.3311 0.6591 0.9271 0>}
  n13 = Add(
   A=LinearInterpolate(A=Multiply(A=@n2.RGB, B=@n8.RGB), B=Multiply(A=@n9.RGB, B=@n10.Result), Alpha=NamedRerouteUsage())
   B=Multiply(A=Power(Base=@n11) {constExponent=8}, B=Multiply(A=@n10.Result, B=@n12.RGB))
  )
  n14 = ScalarParameter() {defaultValue=1, parameterName=Brightness}
  Multiply(A=Multiply(A=Multiply(A=@n13, B=@n14)) {constB=2}, B=@n1.RGB)
  n15 = MaterialFunctionCall() {materialFunction=&LinearGradient}
  n16 = Power(Base=Sine(None=Multiply(A=@n15.VGradient) {constB=0.5})) {constExponent=4}
  n17 = TextureCoordinate() {uTiling=2}
  n18 = TextureSample(UVs=Add(A=Panner(Coordinate=@n17) {speedY=0.1}, B=@n7)) {samplerType=SAMPLERTYPE_LinearGrayscale, texture=/Game/Effects/Fort_Effects/Emotes/UltraLight/Texture/T_DirectionalNoise_Greyscale}
  n19 = Fresnel(ExponentIn=LinearInterpolate(Alpha=@n18.R) {constB=15}, BaseReflectFractionIn=LinearInterpolate(Alpha=@n18.R)) {baseReflectFraction=0, exponent=1}
  n20 = LinearInterpolate(Alpha=@n18.R) {constA=4, constB=15}
  n21 = Saturate(None=Multiply(
    A=MaterialFunctionCall("FadeSensitivity (optional)"=@n20) {materialFunction=&FadeBasedOnViewAngle}.Result
    B=LinearInterpolate(Alpha=@n18.R) {constA=20, constB=0}
   ))
  n22 = DepthFade(Opacity=Add(A=Multiply(A=@n19, B=@n21), B=NamedRerouteUsage()), FadeDistance=ScalarParameter() {defaultValue=5000, parameterName=Depthfade})
  n23 = TextureSample(UVs=Multiply(A=TextureCoordinate(), B=Constant() {r=3})) {samplerType=SAMPLERTYPE_LinearGrayscale, texture=/Game/Effects/Fort_Effects/Emotes/UltraLight/Texture/T_Plasma_A_Greyscale}
  n24 = Multiply(A=@n23.R, B=LinearInterpolate(Alpha=ScalarParameter() {parameterName=Erode}) {constB=5})
  n25 = Multiply(A=Saturate(None=Subtract(A=@n22, B=@n24)), B=ScalarParameter() {defaultValue=1.5, parameterName=OverallOpacity})
  Multiply(A=@n1.A, B=Multiply(A=Multiply(A=@n16, B=@n25), B=NamedRerouteUsage()))
  n26 = NamedRerouteUsage()
  n27 = Power(Base=@n26, Exp=ScalarParameter() {defaultValue=0.5, group=HitWave, parameterName=HitWave-OuterRingDelay}) {constExponent=0.7}
  n28 = Multiply(A=ScalarParameter() {defaultValue=300, group=HitWave, parameterName=HitWave-SizeMultiplierOuterRing}, B=@n27)
  n29 = LinearInterpolate(A=Divide(A=@n28), B=@n28, Alpha=@n23.R)
  n30 = ComponentMask(None=VectorParameter() {group=HitWave, parameterName=HitWave-Loc, defaultValue=<rgba 100 -1400 -150 0>}.RGB) {b=true, g=true, r=true}
  n31 = WorldPosition() {worldPositionShaderOffset=WPT_ExcludeAllShaderOffsets}
  n32 = Saturate(None=LinearInterpolate(A=ScalarParameter() {defaultValue=-0.1, group=HitWave, parameterName=HitWave-InnerRingDelay}, Alpha=@n26) {constA=-0.7})
  n33 = Multiply(A=@n32, B=ScalarParameter() {defaultValue=1000, group=HitWave, parameterName=HitWave-SizeMultiplierInnerRing})
  n34 = LinearInterpolate(A=LinearInterpolate(A=Divide(A=@n33), B=@n33, Alpha=@n23.R), B=@n33, Alpha=@n32)
  n35 = ScalarParameter() {defaultValue=0.25, group=HitWave, parameterName=Hitwave-SphereHard2}
  n36 = Multiply(
   A=SphereMask(
    A=@n31.XYZ
    B=@n30
    Radius=LinearInterpolate(A=@n29, B=@n28, Alpha=@n27)
    Hardness=ScalarParameter() {defaultValue=0.7, group=HitWave, parameterName=Hitwave-SphereHard1}
   ) {hardnessPercent=0}
   B=OneMinus(None=SphereMask(A=@n31.XYZ, B=@n30, Radius=@n34, Hardness=@n35) {hardnessPercent=0})
  )
  n37 = Constant()
  NamedRerouteDeclaration(None=ShadingPathSwitch(Default=@n36, Deferred=@n36, Forward=@n37, Mobile=@n37)) {name=ImpactShockwaveInfo, nodeColor=<rgba 0.2187 1 0 1>}
  n38 = Subtract(A=Time(), B=ScalarParameter() {defaultValue=1, desc=Time-stamp, group=HitWave, parameterName=GameTime})
  n39 = ScalarParameter() {defaultValue=0.5, group=HitWave, parameterName=HitWave-DurationNew}
  NamedRerouteDeclaration(None=Saturate(None=Divide(A=@n38, B=@n39))) {name="Impactinfo(0-1)", nodeColor=<rgba 0.9922 1 0 1>}
  NamedRerouteDeclaration(None=ParticleColor().A) {name=OUT_ParticleAlpha, nodeColor=<rgba 1 0 0.75 1>}
  Power(Base=@n19) {constExponent=1}
  n40 = Power(Base=Saturate(None=NamedRerouteUsage()), Exp=ScalarParameter() {defaultValue=1, group=HitWave, parameterName=Hitwave-pow})
  n41 = Multiply(A=Multiply(A=Saturate(None=@n40), B=ScalarParameter() {defaultValue=45, group=HitWave, parameterName=WPO-AddWave2}), B=VertexNormalWS())
  n42 = Multiply(
   A=SphereMask(A=WorldPosition().XYZ, B=VectorParameter() {group=HitWave, parameterName=HitWave-Loc, defaultValue=<rgba 100 -1400 -150 0>}.RGB) {hardnessPercent=0}
   B=Constant() {r=5}
  )
  n43 = Subtract(A=Time(), B=ScalarParameter() {defaultValue=1, desc=Time-stamp, group=HitWave, parameterName=GameTime})
  n44 = Sine(None=Saturate(None=Divide(A=@n43) {constB=0.2}))
  n45 = Transform(None=Multiply(A=VectorParameter() {parameterName=HitWave-Normal}.RGB, B=Multiply(A=@n42, B=@n44)) {constB=100}) {transformSourceType=TRANSFORMSOURCE_Local}
  n46 = TextureCoordinate()
  n47 = Sine(None=Multiply(A=MaterialFunctionCall(Float2=@n46) {materialFunction=&BreakOutFloat2Components}.G) {constB=0.5})
  n48 = Constant() {r=1}
  n49 = Constant() {r=1.2}
  n50 = Panner(Coordinate=Multiply(A=@n46, B=AppendVector(A=@n48, B=@n49))) {speedX=0.07, speedY=-0.25}
  n51 = Multiply(
   A=VertexNormalWS()
   B=Multiply(
    A=Saturate(None=@n47)
    B=TextureSample(UVs=@n50) {samplerType=SAMPLERTYPE_Grayscale, texture=&LowResBlurredNoise}.RGB
   )
  )
  n52 = Multiply(A=ScalarParameter() {defaultValue=40, parameterName=WPO}, B=DynamicParameter() {paramNames [WPO,Param2,Param3,Param4]}.WPO)
  Reroute(None=Add(A=Add(A=@n41, B=@n45) {constB=0}, B=Multiply(A=@n51, B=@n52)))
  TextureSample() {samplerType=SAMPLERTYPE_Normal, texture=/Game/Global/Textures/Default/Noise/T_SwirlNoise_N}
 ]
}
