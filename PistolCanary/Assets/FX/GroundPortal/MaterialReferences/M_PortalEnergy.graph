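# Translucent, unlit, two-sided particle material for ground portal energy effects, built for Niagara sprites and meshes.
# Distorted, panning noise blends two colors, tinted by particle color with optional time-of-day modulation.
# Opacity combines noise, a vertical gradient mask and optional erosion, depth, camera, view-angle and side fades, then feeds a TemporalAA dither.
# Dynamic parameters drive panning, erosion and texture-based, gradient-weighted vertex offset, which is zeroed on low quality and mobile.
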Material /PistolCanary/Assets/FX/GroundPortal/MaterialReferences/M_PortalEnergy
graph {
 properties {
  bUsedWithMeshParticles=true
  bUsedWithNanite=true
  bUsedWithNiagaraMeshParticles=true
  bUsedWithNiagaraSprites=true
  bUsedWithParticleSprites=true
  blendMode=BLEND_Translucent
  pixelDepthOffsetMode=PDOM_Legacy
  shadingModel=MSM_Unlit
  twoSided=true
  usedShadingModels=Unlit
 }
 nodes [
  n1 = WorldPosition()
  Add(B=@n1.XYZ)
  ComponentMask(None=@n1.XYZ) {r=true}
  MaterialFunctionCall() {materialFunction=&DebugTimeSine}
  n2 = MaterialFunctionCall(
   Color_Morning=ScalarParameter() {defaultValue=1, parameterName=OpacityMorning}
   Color_Day=ScalarParameter() {defaultValue=1, parameterName=OpacityDay}
   Color_Evening=ScalarParameter() {defaultValue=1, parameterName=OpacityEvening}
   Color_Night=ScalarParameter() {defaultValue=1, parameterName=OpacityNight}
  ) {materialFunction=&Color_Time_of_Day}
  n3 = ParticleColor()
  n4 = ComponentMask(None=TextureSample(UVs=TextureCoordinate() {uTiling=2}) {samplerType=SAMPLERTYPE_Normal, texture=&T_Weap_Distort_N_02}.RGB) {g=true, r=true}
  n5 = LinearInterpolate(
   A=ScalarParameter() {defaultValue=8, parameterName=UV_Distortion}
   B=Multiply(A=DynamicParameter() {desc=Niagara, parameterIndex=1}.Param1) {constB=8}
   Alpha=ScalarParameter() {parameterName=UseDPtoHandleUVDistortion}
  )
  n6 = MaterialFunctionCall(Float2=TextureCoordinate()) {materialFunction=&BreakOutFloat2Components}
  n7 = Multiply(A=@n6.R, B=ScalarParameter() {defaultValue=2, parameterName=U_Tile})
  n8 = Power(
   Base=Multiply(A=@n6.G, B=ScalarParameter() {defaultValue=1, parameterName=V_Tile})
   Exp=ScalarParameter() {defaultValue=3, parameterName=V_Tile_Exp}
  ) {constExponent=3}
  n9 = DynamicParameter() {defaultValue=<rgba 10 0 1 1>, paramNames [WPO, Erode, PanSeed_V, CurvePanAnim_V]}
  n10 = ScalarParameter() {parameterName=UseDPtoCurvePanAnim_U}
  n11 = ScalarParameter() {parameterName=SpeedX}
  n12 = AppendVector(A=@n11, B=@n9.PanSeed_V)
  n13 = AppendVector(A=@n11, B=ScalarParameter() {defaultValue=1, parameterName=SpeedY})
  n14 = Panner(
   Coordinate=Add(A=Divide(A=@n4, B=@n5) {constB=8}, B=AppendVector(A=@n7, B=@n8))
   Time=LinearInterpolate(A=Time(), B=@n9.CurvePanAnim_V, Alpha=@n10)
   Speed=LinearInterpolate(A=StaticSwitchParameter(True=@n12, False=@n13) {parameterName=UseDPtoHandlePanSpeedY}, B=@n13, Alpha=@n10)
  ) {speedY=1}
  n15 = TextureSampleParameter2D(UVs=@n14) {parameterName=NoiseB, texture=/Game/Global/Textures/Default/Caustics/T_Turbulence_Seamless}
  n16 = ScalarParameter() {parameterName=Use_TextureRGB}
  n17 = LinearInterpolate(A=@n15.R, B=@n15.R, Alpha=@n16)
  n18 = TextureSampleParameter2D(UVs=@n14) {parameterName=NoiseA, texture=&T_ShieldCaustics}
  n19 = LinearInterpolate(A=@n18.R, B=@n18.R, Alpha=@n16)
  n20 = Multiply(
   A=Desaturation(None=@n3.RGB, Fraction=ScalarParameter() {parameterName=inputDesaturation})
   B=LinearInterpolate(
    A=VectorParameter() {parameterName=ColorLow, defaultValue=<rgba 0.0599 0.0025 0.1198 1>}.RGB
    B=VectorParameter() {parameterName=ColorHigh, defaultValue=<rgba 0.6635 0.2274 1 1>}.RGB
    Alpha=LinearInterpolate(A=@n17, B=@n19)
   )
  )
  MaterialFunctionCall(Emissive=StaticSwitchParameter(True=Multiply(A=@n2.Result, B=@n20), False=@n20) {parameterName=Use_color_time_of_day}) {materialFunction=&MF_FX_EmissiveAdjustments}
  n21 = DepthFade(FadeDistance=ScalarParameter() {parameterName=FadeDistance})
  n22 = Constant() {r=1}
  n23 = TextureSample(UVs=@n14) {texture=&T_NoisePack_01}
  n24 = Power(
   Base=MaterialFunctionCall() {materialFunction=&LinearGradient}.VGradient
   Exp=ScalarParameter() {defaultValue=0.4, group=WPO, parameterName=CenterOffset}
  ) {constExponent=0.5}
  n25 = Constant() {r=0.5}
  n26 = Multiply(A=Add(A=@n23.G, B=@n23.G), B=Power(Base=SphereMask(A=@n24, B=@n25, Radius=@n25) {hardnessPercent=0}) {constExponent=3})
  n27 = StaticSwitchParameter(
   True=MaterialFunctionCall("FadeSensitivity (optional)"=ScalarParameter() {defaultValue=3, parameterName=FadeBasedOnViewAngle}) {materialFunction=&FadeBasedOnViewAngle}.Result
   False=Constant() {r=1}
  ) {parameterName=UseFadeOnViewAngle}
  n28 = VectorParameter() {group=WPO, parameterName=TextureWPO_UVtiling, defaultValue=<rgba 0.99 0.5 0 0>}
  n29 = Panner(Coordinate=Multiply(A=TextureCoordinate() {vTiling=0.5}, B=ComponentMask(None=@n28.RGB) {g=true, r=true})) {speedY=0.1}
  n30 = TextureSample(UVs=@n29) {texture=&T_Smooth_tile}
  n31 = Subtract(
   A=Multiply(
    A=MaterialFunctionCall() {materialFunction=&LinearGradient}.VGradient
    B=LinearInterpolate(Alpha=@n30.R) {constA=0.5}
   )
   B=DynamicParameter() {defaultValue=<rgba 10 0 1 1>, paramNames [WPO, Erode, PanSeed_V, CurvePanAnim_V]}.Erode
  )
  n32 = Multiply(
   A=Multiply(A=@n3.A, B=Multiply(A=@n26, B=@n27))
   B=StaticSwitchParameter(True=Saturate(None=@n31), False=Constant() {r=1}) {parameterName=UseErodeByGradient}
  )
  n33 = StaticSwitchParameter(True=Multiply(A=FeatureLevelSwitch(Default=@n21, ES2=@n22, ES3_1=@n22, SM4_REMOVED=@n21), B=@n32), False=@n32) {parameterName=UseDepthFade}
  n34 = MaterialFunctionCall(Fade Length=ScalarParameter() {defaultValue=5, group=Use Camera Fade, parameterName=CameraDepthFade}) {materialFunction=&CameraDepthFade}
  n35 = Constant() {r=1}
  n36 = LinearInterpolate(
   A=@n33
   B=Multiply(A=@n33, B=FeatureLevelSwitch(Default=@n34.Result, ES2=@n35, ES3_1=@n35))
   Alpha=ScalarParameter() {group=Use Camera Fade, parameterName=UseCameraFade}
  )
  n37 = TextureCoordinate()
  n38 = Sine(None=Divide(A=ComponentMask(None=@n37) {r=true}))
  n39 = ScalarParameter() {defaultValue=1, group=SideGradient, parameterName=SideGradientPower, sortPriority=1}
  n40 = Saturate(None=Multiply(A=Power(Base=@n38, Exp=@n39), B=ScalarParameter() {defaultValue=1, group=SideGradient, parameterName=SideGradientMultiply, sortPriority=2}))
  MaterialFunctionCall(Alpha Threshold=LinearInterpolate(A=@n36, B=Multiply(A=@n36, B=@n40), Alpha=ScalarParameter() {group=SideGradient, parameterName=UseSideGradient, sortPriority=0})) {materialFunction=&DitherTemporalAA_Fortnite}
  n41 = Time()
  n42 = ComponentMask(None=TextureCoordinate()) {r=true}
  Multiply(A=Constant3Vector() {constant=<rgba 0 0 150 0>}, B=Sine(None=Add(A=@n41, B=@n42)))
  n43 = ComponentMask(None=TextureSampleParameter2D(UVs=@n29) {parameterName=WPO, samplerType=SAMPLERTYPE_Normal, texture=/Game/Effects/Fort_Effects/Textures/T_DustDistortion}.RGB) {g=true, r=true}
  n44 = DynamicParameter() {defaultValue=<rgba 10 0 1 1>, paramNames [WPO, Erode, PanSeed_V, CurvePanAnim_V]}
  n45 = ScalarParameter() {defaultValue=0.1, group=WPO, parameterName=WPO_Height}
  n46 = MaterialFunctionCall() {materialFunction=&LinearGradient}
  n47 = LinearInterpolate(
   A=AppendVector(A=Constant2Vector(), B=ScalarParameter() {group=WPO, parameterName=BaseHight})
   B=AppendVector(A=Multiply(A=@n43, B=@n44.WPO) {constB=10}, B=Multiply(A=@n30.R, B=@n45) {constB=0.6})
   Alpha=@n46.VGradient
  )
  n48 = Normalize(VectorInput=Transform(None=@n47) {transformSourceType=TRANSFORMSOURCE_Local})
  n49 = Multiply(
   A=If(A=ScalarParameter() {parameterName=WPO from Local to World Space}, "A > B"=@n48, "A == B"=@n48, "A < B"=Transform(None=@n47)) {constB=0.99}
   B=ScalarParameter() {defaultValue=80, parameterName=WPO_OverallIntensity}
  ) {constB=80}
  n50 = Constant3Vector() {desc=Disable WPO anim in Low}
  n51 = FeatureLevelSwitch(Default=@n49, ES2=@n50, ES3_1=@n50, SM4_REMOVED=@n49, SM5=@n49, SM6=@n49)
  QualitySwitch(Default=@n51, Low=@n50, High=@n51, Medium=@n51)
  n52 = Constant() {r=0.5}
  SphereMask(
   A=Power(Base=@n46.VGradient, Exp=ScalarParameter() {defaultValue=0.4, group=WPO, parameterName=CenterOffset}) {constExponent=0.5}
   B=@n52
   Radius=@n52
  ) {hardnessPercent=0}
 ]
}
