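# Translucent unlit particle material that draws a swirling black-hole portal with an edge ring, nebula and stars.
# A radial sphere-mask distortion warps rotating nebula and star UVs, with an optional curve-atlas color ramp and HDR boosts for ring and stars.
# An inner mask blends in a core color, amplified by DepthFade near intersecting geometry, under time-of-day and particle color tinting.
# Opacity combines particle alpha and a soft circle with optional depth fade and center hole.
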Material /PistolCanary/Assets/FX/GroundPortal/MaterialReferences/M_Portal_BlackHole
graph {
 properties {
  bUsedWithNiagaraMeshParticles=true
  bUsedWithNiagaraSprites=true
  bUsedWithParticleSprites=true
  blendMode=BLEND_Translucent
  pixelDepthOffsetMode=PDOM_Legacy
  shadingModel=MSM_Unlit
  usedShadingModels=Unlit
 }
 nodes [
  n1 = ParticleColor()
  n2 = Constant() {r=1}
  n3 = ScalarParameter() {defaultValue=0.25, parameterName=InnerHoleRadius}
  n4 = MaterialFunctionCall(UVs=TextureCoordinate(), Texture Scale=Multiply(A=ScalarParameter() {defaultValue=2.15, parameterName=RingSize}, B=@n3)) {materialFunction=&ScaleUVsByCenter}
  n5 = Multiply(A=Time(), B=ScalarParameter() {defaultValue=2, parameterName=MasterPanSpeed})
  n6 = Multiply(A=Subtract(A=TextureCoordinate()) {constB=0.5}) {constB=2}
  n7 = SphereMask(
   A=TextureCoordinate()
   B=Constant() {r=0.5}
   Radius=ScalarParameter() {defaultValue=0.4, parameterName=OuterHoleRadius}
   Hardness=ScalarParameter() {defaultValue=0.4, parameterName=OuterHoleHardness}
  )
  n8 = Power(Base=@n7, Exp=ScalarParameter() {defaultValue=4, parameterName=DistortionPower})
  n9 = Multiply(A=Multiply(A=@n8, B=@n6), B=ScalarParameter() {defaultValue=-0.25, parameterName=Distortion})
  n10 = Add(A=Multiply(A=@n9) {constB=0.5}, B=TextureCoordinate() {uTiling=1.7, vTiling=1.7})
  n11 = TextureSample(UVs=Panner(Coordinate=Rotator(Coordinate=@n10) {centerX=0.35, centerY=0.35, speed=0.025}, Time=@n5) {speedX=0.05, speedY=-0.03}) {samplerType=SAMPLERTYPE_Masks, texture=/Game/Characters/Player/Male/Medium/Bodies/M_Med_Soldier_04/Skins/Galaxy/Textures/Celestial_T_Stars}
  n12 = Multiply(A=@n6, B=Multiply(A=@n11.G, B=@n11.G))
  n13 = TextureSample(UVs=Add(A=Rotator(Coordinate=@n4.UVs, Time=@n5) {speed=-0.07}, B=Multiply(A=@n12) {constB=0.12})) {texture=/Game/Athena/Prototype/Blueprints/NightNight/FX/Dot/T_NightNight_Curse}
  n14 = LinearInterpolate(
   A=VectorParameter() {parameterName=HoleEdgeGlow2, defaultValue=<rgba 0.625 0 1.5 0>}.RGB
   B=VectorParameter() {parameterName=HoleEdgeGlow, defaultValue=<rgba 0.0407 2 1.7204 0>}.RGB
   Alpha=Power(Base=@n7) {constExponent=7}
  )
  n15 = Add(A=@n9, B=TextureCoordinate() {uTiling=0.45, vTiling=0.45})
  n16 = TextureSample(UVs=Rotator(Coordinate=Add(A=@n15) {constB=0.15}, Time=@n5) {centerX=0.35, centerY=0.4, speed=-0.04}) {texture=/Game/Effects/Fort_Effects/Textures/Tile/T_Portal_Space_D}
  n17 = CurveAtlasRowParameter(CurveTime=ComponentMask(None=RgbToHsv!(None=@n16.RGB)) {b=true}) {
   atlas=/Game/Accessories/FORT_Backpacks/Curves/CA_Backpack_Curves
   curve=/Game/Accessories/FORT_Backpacks/Curves/C_Backpack_HauntKoi_Ramp
   parameterName=Curve Atlas
  }
  n18 = LinearInterpolate(A=Add(A=@n16.RGB) {constB=0.05}, B=@n17, Alpha=ScalarParameter() {parameterName=Use Color Ramp}) {constAlpha=0}
  n19 = Add(
   A=Multiply(A=VectorParameter() {parameterName=RingColor, defaultValue=<rgba 1.2869 1.5249 3 0>}.RGB, B=Power(Base=@n13.G) {constExponent=1.5})
   B=Multiply(A=Power(Base=@n7) {constExponent=5}, B=Multiply(A=@n14, B=@n18))
  )
  n20 = Saturate(None=@n19)
  n21 = ScalarParameter() {defaultValue=2, parameterName=GlowHDRMult}
  n22 = Multiply(A=@n18, B=ScalarParameter() {defaultValue=1.2, parameterName=NebulaScale}) {constB=1.2}
  n23 = Multiply(A=@n9, B=ScalarParameter() {defaultValue=2, parameterName=StarsDistortMult}) {constB=-0.25}
  n24 = Add(A=@n23, B=TextureCoordinate())
  n25 = TextureSample(UVs=Panner(Coordinate=Rotator(Coordinate=@n24) {centerX=0.26, centerY=0.24, speed=-0.03}, Time=@n5) {speedX=-0.004, speedY=0.008}) {samplerType=SAMPLERTYPE_Masks, texture=/Game/Packages/Fortress_Sky/Textures/T_Sky_Stars_Sharp}
  n26 = Add(A=@n23, B=TextureCoordinate() {uTiling=0.6, vTiling=0.6})
  n27 = TextureSample(UVs=Panner(Coordinate=Rotator(Coordinate=@n26) {centerX=0.14, centerY=0.16, speed=-0.025}, Time=@n5) {speedX=0.002, speedY=-0.001}) {samplerType=SAMPLERTYPE_Masks, texture=&T_Sky_Stars_Sharp}
  n28 = Power(Base=Multiply(A=Add(A=@n25.G, B=@n27.G), B=@n11.G), Exp=ScalarParameter() {defaultValue=2.5, parameterName=StarsPower})
  n29 = Multiply(
   A=Add(A=@n8)
   B=MaterialFunctionCall(LDR Input=Saturate(None=@n28), HDR Tint=@n28, HDR Intensity Multiplier=ScalarParameter() {defaultValue=200, parameterName=StarsHDRMult}) {materialFunction=&DeriveHDRfromLDR}.RemappedRange
  )
  n30 = Multiply(
   A=MaterialFunctionCall(Value_Morning=@n2, Value_Day=ScalarParameter() {defaultValue=1.7, parameterName=DayMult}, Value_Evening=@n2, Value_Night=@n2) {materialFunction=&Scalar_Time_of_Day}.Result
   B=Add(
    A=MaterialFunctionCall(LDR Input=@n20, HDR Tint=@n19, HDR Intensity Multiplier=@n21) {materialFunction=&DeriveHDRfromLDR}.RemappedRange
    B=Add(A=@n22, B=@n29)
   )
  )
  n31 = VectorParameter() {parameterName=Color_Core, defaultValue=<rgba 1 0 1 0>}
  n32 = Power(Base=SphereMask(A=TextureCoordinate(), B=Constant() {r=0.5}, Radius=@n3, Hardness=ScalarParameter() {defaultValue=0.7, parameterName=InnerHoleHardness})) {constExponent=4}
  n33 = ScalarParameter() {defaultValue=1, parameterName=RDBA_HighlightGlow_Mult}
  n34 = DepthFade(FadeDistance=ScalarParameter() {defaultValue=512, parameterName=RDBA_HighlightGlow_Strength}) {fadeDistanceDefault=512}
  n35 = Add(A=@n32, B=LinearInterpolate(B=Multiply(A=@n32, B=@n33), Alpha=OneMinus(None=@n34)))
  MaterialFunctionCall(Emissive=Multiply(A=@n1.RGB, B=LinearInterpolate(A=@n30, B=@n31.RGB, Alpha=@n35) {constB=0})) {materialFunction=&MF_FX_EmissiveAdjustments}
  n36 = TextureCoordinate()
  n37 = Constant() {r=0.5}
  n38 = Constant() {r=0.4}
  n39 = Multiply(A=@n1.A, B=Power(Base=SphereMask(A=@n36, B=@n37, Radius=@n37, Hardness=@n38) {hardnessPercent=40}))
  n40 = DepthFade(FadeDistance=ScalarParameter() {parameterName=FadeDistance})
  n41 = MaterialFunctionCall(Radius=Constant() {r=0.24}, Density=Constant() {r=3}) {materialFunction=&RadialGradientExponential}
  Multiply(
   A=LinearInterpolate(A=@n39, B=Multiply(A=@n39, B=@n40), Alpha=ScalarParameter() {parameterName=Use Depth Fade})
   B=LinearInterpolate(
    A=Constant() {r=1}
    B=SmoothStep(Value=@n41.RadialGradientExponential) {constMax=0, constMin=0.7}
    Alpha=ScalarParameter() {parameterName=HoleMask}
   )
  )
 ]
}
